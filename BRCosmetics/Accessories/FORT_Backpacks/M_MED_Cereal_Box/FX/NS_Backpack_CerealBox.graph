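# Niagara effect for the cereal box backpack cosmetic that throws short-lived bursts of colored cereal sprites up and out of the box.
# User.Colors selects one of three bands on a stepped color curve, and each particle takes a random color within it.
# Sprites fall under gravity with size-based drag and shrink over life.
# Cereal_World particles live longer than Cereal_Local ones and add inherited source movement scaled by a curve.
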NiagaraSystem /BRCosmetics/Accessories/FORT_Backpacks/M_MED_Cereal_Box/FX/NS_Backpack_CerealBox
dependencies {
 usedDataInterfaces [&NiagaraDataInterfaceColorCurve, &NiagaraDataInterfaceArrayFloat, &NiagaraDataInterfaceCurve]
 usedDynamicInputs []
 usedModules [&SystemState]
}
emitters {
 Cereal_Local {
  emitterUpdateScript [
   &EmitterState()
   &SpawnBurst_Instantaneous(Spawn Count=10, Use Spawn Probability=false, UseLoopCountLimit=false)
  ]
  particleSpawnScript [
   &InitializeParticle(
    Color=&ColorFromCurve(
     ColorCurve=DataInterface() {propertyValues {
       AlphaCurve {keys [<at 0 1 RCIM_Constant>]}
       BlueCurve {keys [<at 0 1 RCIM_Constant>,<at 0.1739 0.0551 RCIM_Constant>,<at 0.2302 0.4418 RCIM_Constant>,<at 0.2857 0.9601 RCIM_Constant>,<at 0.33 0 RCIM_Constant>,<at 0.4264 0 RCIM_Constant>,<at 0.4983 0.6094 RCIM_Constant>,<at 0.5546 1 RCIM_Constant>,<at 0.66 0.533 RCIM_Constant>,<at 0.8835 0.13 RCIM_Constant>,<at 0.9481 1 RCIM_Constant>]}
       GreenCurve {keys [<at 0 0.8658 RCIM_Constant>,<at 0.1739 0.3212 RCIM_Constant>,<at 0.2302 0.0248 RCIM_Constant>,<at 0.2857 0.015 RCIM_Constant>,<at 0.33 0.5573 RCIM_Constant>,<at 0.4264 0.2124 RCIM_Constant>,<at 0.4983 0.836 RCIM_Constant>,<at 0.5546 0 RCIM_Constant>,<at 0.66 0.048 RCIM_Constant>,<at 0.8835 0.0009 RCIM_Constant>,<at 0.9481 0.54 RCIM_Constant>]}
       RedCurve {keys [<at 0 0 RCIM_Constant>,<at 0.1739 0.0139 RCIM_Constant>,<at 0.2302 0.8388 RCIM_Constant>,<at 0.2857 0.094 RCIM_Constant>,<at 0.33 1 RCIM_Constant>,<at 0.4264 1 RCIM_Constant>,<at 0.4983 1 RCIM_Constant>,<at 0.5546 0.2578 RCIM_Constant>,<at 0.66 0.1225 RCIM_Constant>,<at 0.8835 0.5365 RCIM_Constant>,<at 0.9481 0.0381 RCIM_Constant>]}
      }}
     CurveIndex=&RandomRangeFloat(
      Minimum=&SelectFloatFromArray(
       Direct Array Index=&MakeIntFromFloat(Float=&Add_Float(A=$User.Colors, B=0))
       Direct Array Mode=::NewEnumerator1
       Float Selection Array=DataInterface() {propertyValues {FloatData [0,0.33,0.66]}}
      )
      Maximum=&SelectFloatFromArray(
       Direct Array Index=&MakeIntFromFloat(Float=&Add_Float(A=$User.Colors, B=0))
       Direct Array Mode=::NewEnumerator1
       Float Selection Array=DataInterface() {propertyValues {FloatData [0.32,0.65,1]}}
      )
     )
    )
    Lifetime=&RandomRangeFloat(Minimum=0.3, Maximum=0.7)
    Sprite Rotation Mode=::NewEnumerator2
    Sprite Size Mode=::NewEnumerator3
    Uniform Sprite Size=&RandomRangeFloat(Minimum=2, Maximum=7)
    UsePositionOffset=false
    Write Material Random=true
   )
   &BoxLocation(Box Offset=<0 -7 5>, Box Size=<25 5 10>, MaskSpawnGroup=false, Module.Override Seed=false, Rotate Box=false, Surface Only=false)
   &AddVelocityFromPoint(
    Module.Use Falloff Distance=false
    Origin Offset=<0 -3 -5>
    Velocity Strength=&Multiply_Float(
     A=&UniformRangedFloat(Minimum=75, Maximum=150)
     B=&FloatFromCurve(FloatCurve=DataInterface() {propertyValues {Curve {keys [
          {arriveTangent=-1.8998, interpMode=RCIM_Cubic, leaveTangent=-1.8998, tangentMode=RCTM_User, time=0, value=1}
          {interpMode=RCIM_Cubic, tangentMode=RCTM_User, time=1, value=0}
         ]}}})
    )
   )
   &AddVelocity(Velocity=&RandomRangeVector(Minimum=<0 0 100>, Maximum=<1 1 175>))
   /Niagara/Modules/Update/SubUV/SubUVAnimation(SubUV Animation Mode=::NewEnumerator1)
  ]
  particleUpdateScript [
   &GravityForce(Gravity=<0 0 -700>)
   &Drag(
    Drag=&Lerp_Float(
     A=1.5
     B=0.7
     Alpha=&NormalizeFloat(Value=&MakeFloatFromVector2d(Vector 2D=$Particles.Initial.SpriteSize), Min=2, Max=7)
    )
    Use Linear Drag=true
    Use Rotational Drag=false
   )
   &SolveForcesAndVelocity(Clamp Velocity=false, Limit Acceleration=false)
   &ParticleState()
   &ScaleSpriteSize(Uniform Curve Sprite Scale=DataInterface() {propertyValues {Curve {keys [{interpMode=RCIM_Cubic, tangentMode=RCTM_User, time=0, value=1}, {interpMode=RCIM_Cubic, tangentMode=RCTM_User, time=1, value=0}]}}})
   &SpriteRotationRate()
  ]
  rendererClasses [&NiagaraSpriteRendererProperties]
  renderers [{rendererClass=&NiagaraSpriteRendererProperties}]
 }
 Cereal_World {
  emitterUpdateScript [
   &EmitterState()
   &SpawnBurst_Instantaneous(Spawn Count=10, Use Spawn Probability=false, UseLoopCountLimit=false)
  ]
  particleSpawnScript [
   &InitializeParticle(
    Color=&ColorFromCurve(
     ColorCurve=DataInterface() {propertyValues {
       AlphaCurve {keys [<at 0 1 RCIM_Constant>]}
       BlueCurve {keys [<at 0 1 RCIM_Constant>,<at 0.1739 0.0551 RCIM_Constant>,<at 0.2302 0.4418 RCIM_Constant>,<at 0.2857 0.9601 RCIM_Constant>,<at 0.33 0 RCIM_Constant>,<at 0.5399 0 RCIM_Constant>,<at 0.5978 0.6094 RCIM_Constant>,<at 0.66 0.533 RCIM_Constant>,<at 0.8835 0.13 RCIM_Constant>,<at 0.9481 1 RCIM_Constant>]}
       GreenCurve {keys [<at 0 0.8658 RCIM_Constant>,<at 0.1739 0.3212 RCIM_Constant>,<at 0.2302 0.0248 RCIM_Constant>,<at 0.2857 0.015 RCIM_Constant>,<at 0.33 0.5573 RCIM_Constant>,<at 0.5399 0.2124 RCIM_Constant>,<at 0.5978 0.836 RCIM_Constant>,<at 0.66 0.048 RCIM_Constant>,<at 0.8835 0.0009 RCIM_Constant>,<at 0.9481 0.54 RCIM_Constant>]}
       RedCurve {keys [<at 0 0 RCIM_Constant>,<at 0.1739 0.0139 RCIM_Constant>,<at 0.2302 0.8388 RCIM_Constant>,<at 0.2857 0.094 RCIM_Constant>,<at 0.33 1 RCIM_Constant>,<at 0.5399 1 RCIM_Constant>,<at 0.5978 1 RCIM_Constant>,<at 0.66 0.1225 RCIM_Constant>,<at 0.8835 0.5365 RCIM_Constant>,<at 0.9481 0.0381 RCIM_Constant>]}
      }}
     CurveIndex=&RandomRangeFloat(
      Minimum=&SelectFloatFromArray(
       Direct Array Index=&MakeIntFromFloat(Float=&Add_Float(A=$User.Colors, B=0))
       Direct Array Mode=::NewEnumerator1
       Float Selection Array=DataInterface() {propertyValues {FloatData [0,0.33,0.66]}}
      )
      Maximum=&SelectFloatFromArray(
       Direct Array Index=&MakeIntFromFloat(Float=&Add_Float(A=$User.Colors, B=0))
       Direct Array Mode=::NewEnumerator1
       Float Selection Array=DataInterface() {propertyValues {FloatData [0.32,0.65,1]}}
      )
     )
    )
    Lifetime=&RandomRangeFloat(Minimum=0.7, Maximum=1.3)
    Sprite Rotation Mode=::NewEnumerator2
    Sprite Size Mode=::NewEnumerator3
    Uniform Sprite Size=&RandomRangeFloat(Minimum=2, Maximum=7)
    UsePositionOffset=false
    Write Material Random=true
   )
   &BoxLocation(Box Offset=<0 -7 5>, Box Size=<25 5 10>, MaskSpawnGroup=false, Module.Override Seed=false, Rotate Box=false, Surface Only=false)
   &AddVelocityFromPoint(
    Module.Use Falloff Distance=false
    Origin Offset=<0 -3 -5>
    Velocity Strength=&Multiply_Float(
     A=&UniformRangedFloat(Minimum=75, Maximum=150)
     B=&FloatFromCurve(FloatCurve=DataInterface() {propertyValues {Curve {keys [
          {arriveTangent=-1.8998, interpMode=RCIM_Cubic, leaveTangent=-1.8998, tangentMode=RCTM_User, time=0, value=1}
          {interpMode=RCIM_Cubic, tangentMode=RCTM_User, time=1, value=0}
         ]}}})
    )
   )
   &AddVelocity(Velocity=&RandomRangeVector(Minimum=<0 0 25>, Maximum=<1 1 200>))
   &SubUVAnimation(SubUV Animation Mode=::NewEnumerator1)
  ]
  particleUpdateScript [
   &GravityForce(Gravity=<0 0 -750>)
   &Drag(
    Drag=&Lerp_Float(
     A=1.5
     B=0.7
     Alpha=&NormalizeFloat(Value=&MakeFloatFromVector2d(Vector 2D=$Particles.Initial.SpriteSize), Min=2, Max=7)
    )
    Use Linear Drag=true
    Use Rotational Drag=false
   )
   &InheritSourceMovement(
    Applied Position Scale=&VectorFromFloat(Value=&FloatFromCurve(FloatCurve=DataInterface() {propertyValues {Curve {keys [
          {arriveTangent=-1.7643, interpMode=RCIM_Cubic, leaveTangent=-1.7643, tangentMode=RCTM_User, time=0, value=1}
          {interpMode=RCIM_Cubic, tangentMode=RCTM_User, time=1, value=0}
         ]}}}))
    ApplyForce=false
    ApplyPosition=true
    ApplyPositionFalloff=false
   )
   &SolveForcesAndVelocity(Clamp Velocity=false, Limit Acceleration=false)
   &ParticleState()
   &ScaleSpriteSize(Uniform Curve Sprite Scale=DataInterface() {propertyValues {Curve {keys [{interpMode=RCIM_Cubic, tangentMode=RCTM_User, time=0, value=1}, {interpMode=RCIM_Cubic, tangentMode=RCTM_User, time=1, value=0}]}}})
   &SpriteRotationRate()
  ]
  rendererClasses [&NiagaraSpriteRendererProperties]
  renderers [{rendererClass=&NiagaraSpriteRendererProperties}]
 }
}
summary {
 systemName=NS_Backpack_CerealBox
 emitters [
  {emitterName=Cereal_Local, rendererClasses [&NiagaraSpriteRendererProperties]}
  {emitterName=Cereal_World, rendererClasses [&NiagaraSpriteRendererProperties]}
 ]
 userVariables [User.Colors {type=NiagaraFloat, defaultValue=0, description=""}]
}
