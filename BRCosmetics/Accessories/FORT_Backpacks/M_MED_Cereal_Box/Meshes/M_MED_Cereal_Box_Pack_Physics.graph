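# Physics asset for the cereal box back bling, defining bodies and joints for its four dynamic lid flaps.
# Each lid bone carries one or two capsules, with longer ones on flaps A and D and shorter ones on B and C.
# All four flaps constrain directly to the pack body, with twist and swing2 limits at zero.
# Each acts as a hinge on swing1 with its own angle limit, widest on flap A.
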PhysicsAsset /BRCosmetics/Accessories/FORT_Backpacks/M_MED_Cereal_Box/Meshes/M_MED_Cereal_Box_Pack_Physics
bodies {
 dyn_lid_A [
  {shapeType=Capsule, length=25, radius=1, shapeName=dyn_lid_A_capsule, center=<4.0875 0 0.0354>}
  {shapeType=Capsule, length=25, radius=1, shapeName=dyn_lid_A_capsule, center=<2.3689 0 0.0354>}
 ]
 dyn_lid_B [
  {shapeType=Capsule, length=5.5734, radius=0.6044, shapeName=dyn_lid_B_capsule, center=<3.6 -0.5 -0.2>}
  {shapeType=Capsule, length=5.509, radius=0.8, shapeName=dyn_lid_B_capsule, center=<2.2433 0 -0.1067>}
 ]
 dyn_lid_C [{shapeType=Capsule, length=6.2474, radius=0.6072, shapeName=dyn_lid_C_capsule, center=<3.1268 -0.3 0>}]
 dyn_lid_D [
  {shapeType=Capsule, length=25, radius=1, shapeName=dyn_lid_D_capsule, center=<4.0829 0 0>}
  {shapeType=Capsule, length=25, radius=1, shapeName=dyn_lid_D_capsule, center=<2.422 0 0>}
 ]
 pack [{shapeType=Capsule, length=1, radius=1}]
}
constraints [
 {bone1Name=dyn_lid_A, bone2Name=pack, swing1LimitDegrees=130, swing2LimitDegrees=0, twistLimitDegrees=0}
 {bone1Name=dyn_lid_B, bone2Name=pack, swing1LimitDegrees=60, swing2LimitDegrees=0, twistLimitDegrees=0}
 {bone1Name=dyn_lid_D, bone2Name=pack, swing1LimitDegrees=80, swing2LimitDegrees=0, twistLimitDegrees=0}
 {bone1Name=dyn_lid_C, bone2Name=pack, swing1LimitDegrees=50, swing2LimitDegrees=0, twistLimitDegrees=0}
]
